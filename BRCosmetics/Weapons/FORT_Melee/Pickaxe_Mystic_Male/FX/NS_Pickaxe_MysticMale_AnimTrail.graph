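# Niagara weapon trail for the male Mystic pickaxe, drawing an orange HDR air trail through two ribbon renderers on one emitter.
# AnimTrail modules spawn and update ribbon particles from skeletal mesh bones or sockets, using distance traveled as ribbon UV.
# Alpha combines a lifetime fade with an execution-index curve that tapers both ends, while time of day scales brightness.
# Separate Symbol scale and color attributes derive from ribbon width and a boosted, quicker-fading particle color.
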NiagaraSystem /BRCosmetics/Weapons/FORT_Melee/Pickaxe_Mystic_Male/FX/NS_Pickaxe_MysticMale_AnimTrail
dependencies {usedDataInterfaces [&NiagaraDataInterfaceCurve], usedDynamicInputs [], usedModules [&SystemState]}
emitters {Air_Trail {
  emitterUpdateScript [
   &EmitterState(Life Cycle Mode=::NewEnumerator1, Loop Behavior=::NewEnumerator1, Loop Duration=0.15, UseLoopDelay=true)
   &FN_SpawnAnimTrail_v2()
   &SpawnPerUnit(
    Module.Use Max Movement Threshold=false
    Module.Use Movement Tolerance=false
    Spawn Probability=0
    Spawn Spacing=1899999981993984
    Use Spawn Probability=true
   )
  ]
  particleSpawnScript [
   /Game/Effects/Niagara/Modules/AnimTrail/FN_InitializeAnimTrail/FN_InitializeAnimTrailBoneIndecies(First Bone or Socket Index=1)
   /Game/Effects/Niagara/Modules/AnimTrail/FN_InitializeAnimTrail/FN_InitializeAnimTrail(
    AnimTrailWidth=1
    Color=&RandomRangeLinearColor(Minimum=<rgba 25 2.9167 0 1>, Maximum=<rgba 15 2.5 0 1>)
    Lifetime=0.3
    Position Offset:Vector3f=&Lerp_Vector(A=<0 0 0>, B=<0 0 0>, Alpha=&MakeFloatFromInt(Int=$Particles.SpawnGroup))
   )
   &DynamicMaterialParameters(Module.Param0WriteEnabled=false, Module.Param1WriteEnabled=false, Module.Param2WriteEnabled=false, Module.Param3WriteEnabled=false)
   /SetVariables(Particles.RibbonUVDistance:NiagaraFloat=$Emitter.DistanceTraveled) {isSetParameters=true}
   /SetVariables(Particles.Symbol Scale:NiagaraFloat=&Multiply_Float(A=$Particles.RibbonWidth, B=0.9)) {isSetParameters=true}
  ]
  particleUpdateScript [
   &FN_UpdateAnimTrail_v2()
   &ScaleRibbonWidth(Ribbon Width Scale=&FloatFromCurve(FloatCurve=DataInterface() {propertyValues {Curve {keys [{arriveTangent=-0.8703, time=0.2, value=1}, <at 1 1>]}}})) {enabled=false}
   /SetVariables() {isSetParameters=true}
   &ScaleColor(
    Scale Alpha=&Multiply_Float(
     A=&FloatFromCurve(FloatCurve=DataInterface() {propertyValues {Curve {keys [{arriveTangent=-0.687, interpMode=RCIM_Cubic, leaveTangent=-0.687, tangentMode=RCTM_User, time=0, value=1}, <at 1 0>]}}})
     B=&FloatFromCurve(
      FloatCurve=DataInterface() {propertyValues {Curve {keys [
          {
           arriveTangent=0.7552
           arriveTangentWeight=0.3153
           interpMode=RCIM_Cubic
           leaveTangent=0.7552
           leaveTangentWeight=0.3153
           tangentMode=RCTM_User
           tangentWeightMode=RCTWM_WeightedBoth
           time=0
           value=0
          }
          {arriveTangentWeight=0.2703, interpMode=RCIM_Cubic, leaveTangentWeight=0.2703, tangentMode=RCTM_User, time=0.5, value=1}
          {
           arriveTangent=-1.115
           arriveTangentWeight=0.6059
           interpMode=RCIM_Cubic
           leaveTangent=-1.115
           leaveTangentWeight=0.6059
           tangentMode=RCTM_User
           tangentWeightMode=RCTWM_WeightedBoth
           time=1
           value=0
          }
         ]}}}
      CurveIndex=&ReturnNormalizedExecIndex()
     )
    )
    Scale RGB=&VectorFromFloat(Value=&FNTimeOfDayScalar(Night Scale=0.35, Morning Scale=0.45, Day Scale=0.6, Evening Scale=0.375))
    ScaleA=true
    ScaleRGB=true
   )
   &SpriteFacingAndAlignment(Sprite Facing=$Particles.RibbonFacing, WriteAlignment=false, WriteFacing=true)
   /SetVariables(Particles.Symbol Color:LinearColor=&MultiplyLinearColor(
     Linear Color=$Particles.Color
     Scale Factor=&MakeFloat4FromVectorAndFloat(
      "Vector (RGB)"=&VectorFromFloat(Value=250)
      "Float (Alpha)"=&FloatFromCurve(FloatCurve=DataInterface() {propertyValues {Curve {keys [
           {arriveTangent=-3.5107, interpMode=RCIM_Cubic, leaveTangent=-3.5107, tangentMode=RCTM_User, time=0, value=1}
           {arriveTangent=0.1983, interpMode=RCIM_Cubic, leaveTangent=0.1983, tangentMode=RCTM_User, time=0.35, value=0}
          ]}}})
     )
    )) {isSetParameters=true}
  ]
  rendererClasses [&NiagaraRibbonRendererProperties]
  renderers [
   {rendererClass=&NiagaraRibbonRendererProperties}
   {rendererClass=&NiagaraRibbonRendererProperties, rendererIndex=1}
  ]
 }}
summary {
 systemName=NS_Pickaxe_MysticMale_AnimTrail
 emitters [{emitterName=Air_Trail, rendererClasses [&NiagaraRibbonRendererProperties]}]
 userVariables [
  User.HasHitThing {type=NiagaraBool}
  User.RibbonIDOffset {type=NiagaraInt32}
  User.SkeletalMesh {
   type=NiagaraDataInterfaceSkeletalMesh
   defaultValue {
    struct=NiagaraExt_VariableValue_DataInterface
    value {dataInterface=&self:NiagaraDataInterfaceSkeletalMesh_0}
   }
  }
 ]
}
