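# Niagara system for the swing effect of the Mystic Male pickaxe, in glowing orange.
# Symbols spawns short-lived, spinning SubUV sprites along a local-space Hermite spline, using one random value for both position and tapered spread.
# Woosh bursts a faint mesh after a brief delay, fading it in and out.
# Sparks, which shares the spline placement, and the component-renderer Ref emitter are disabled.
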NiagaraSystem /BRCosmetics/Weapons/FORT_Melee/Pickaxe_Mystic_Male/FX/NS_Pickaxe_Mystic_Male_Swing
dependencies {
 usedDataInterfaces [&NiagaraDataInterfaceCurve, &NiagaraDataInterfaceColorCurve]
 usedDynamicInputs []
 usedModules [&SystemState]
}
emitters {
 Ref {
  bEnabled=false
  emitterUpdateScript [
   &EmitterState()
   &SpawnBurst_Instantaneous(Use Spawn Probability=false, UseLoopCountLimit=false)
  ]
  particleSpawnScript [&InitializeParticle(Color Mode=::NewEnumerator0, Lifetime=5, UsePositionOffset=false, Write Material Random=true)]
  particleUpdateScript [&ParticleState()]
  rendererClasses [&NiagaraComponentRendererProperties]
  renderers [{rendererClass=&NiagaraComponentRendererProperties, rendererIndex=0}]
 }
 Sparks {
  bEnabled=false
  emitterUpdateScript [
   &EmitterState()
   &SpawnRate(SpawnRate=300, Use Spawn Probability=false)
  ]
  particleSpawnScript [
   &InitializeParticle(
    Color=<rgba 25 2.9167 0 1>
    Lifetime Max=0.25
    Lifetime Min=0.1
    Lifetime Mode=::NewEnumerator1
    Mass Mode=::NewEnumerator3
    Sprite Rotation Mode=::NewEnumerator2
    Sprite Size Max=<2 6>
    Sprite Size Min=<4 8>
    Sprite Size Mode=::NewEnumerator2
    UsePositionOffset=false
    Write Material Random=true
   )
   /SetVariables(Particles.MaterialRandom:NiagaraFloat=&RandomRangeFloat(Minimum=0.75, Maximum=1)) {isSetParameters=true, module=SetVariables_#1}
   &SphereLocation(
    MaskSpawnGroup=false
    Override Seed=false
    Sphere Radius=8
    Surface Only=false
    UseNonUniformScale=false
    UseOffset=false
    UseSphereOrientation=false
   ) {enabled=false}
   &AddVelocityInCone(
    Cone Angle=20
    Cone Axis=<0 0 1>
    Module.Use Velocity Falloff On Cone Axis=false
    Velocity Distribution Along Cone Axis=0.75
    Velocity Strength=&RandomRangeFloat(Minimum=500, Maximum=850)
   ) {enabled=false}
   /SetVariables(Particles.Herm Thickness:NiagaraFloat=&FloatFromCurve(
     FloatCurve=DataInterface() {propertyValues {Curve {keys [
         {arriveTangent=0.6937, interpMode=RCIM_Cubic, leaveTangent=0.6937, tangentMode=RCTM_User, time=0, value=0.075}
         {arriveTangent=-2.4361, interpMode=RCIM_Cubic, leaveTangent=-2.4361, tangentMode=RCTM_User, time=1, value=0}
        ]}}}
     CurveIndex=$Particles.MaterialRandom
    )) {isSetParameters=true, module=SetVariables_#2}
   /Game/Effects/Niagara/Modules/Curves/FN_HermiteSpline(
    Compute Acceleration=true
    Compute Velocity=true
    Direction End:Vector3f=&TransformVector(Vector=<-60 0 -80>, Destination Space=::Simulation)
    Direction Start:Vector3f=&TransformVector(Vector=<0 0 150>, Destination Space=::Simulation)
    Position Blend:NiagaraFloat=$Particles.MaterialRandom
    Position End:Vector3f=/Niagara/DynamicInputs/Transforms/TransformPosition(Position=<13 0 67>, Source Space=::Local, Destination Space=::Simulation)
    Position Start:Vector3f=&TransformPosition(Position=<0 0 -7>, Source Space=::Local, Destination Space=::Simulation)
    Write Particles Position=true
   )
   &SphereLocation(
    MaskSpawnGroup=false
    Non Uniform Scale=&MakeVector(X=$Particles.Herm Thickness, Y=0.1, Z=$Particles.Herm Thickness)
    Offset=<0 0.5 0>
    Override Seed=false
    Sphere Radius=10
    Surface Only=false
    UseNonUniformScale=true
    UseOffset=true
    UseSphereOrientation=false
   ) {module=SphereLocation001}
   &AddVelocityInCone(
    Cone Angle=20
    Cone Axis=<1 0 1>
    Module.Use Velocity Falloff On Cone Axis=false
    Velocity Strength=&Multiply_Float(A=&RandomRangeFloat(Minimum=300, Maximum=500), B=0.22)
   ) {module=AddVelocityInCone001}
   &InheritVelocity(Inherited Velocity Amount Scale=<0.25 0.25 0.25>, Speed Limit=false, UseSpeedThreshold=false)
  ]
  particleUpdateScript [
   &ParticleState()
   &GravityForce() {enabled=false}
   &Drag(Drag=0.1, Use Linear Drag=true, Use Rotational Drag=false)
   &ScaleColor(
    Scale Alpha=&FloatFromCurve(FloatCurve=DataInterface() {propertyValues {Curve {keys [<at -0.0046 1.0113 RCIM_Cubic>,<at 1.0011 0.002 RCIM_Cubic>]}}})
    ScaleA=true
    ScaleRGB=true
   )
   &CurlNoiseForce(Noise Frequency=120, Noise Quality / Cost=::NewEnumerator0, Noise Strength=250, Pan Noise=false)
   &SolveForcesAndVelocity(Clamp Velocity=false, Limit Acceleration=false)
  ]
  rendererClasses [&NiagaraSpriteRendererProperties]
  renderers [{rendererClass=&NiagaraSpriteRendererProperties}]
 }
 Symbols {
  emitterUpdateScript [
   &EmitterState(Life Cycle Mode=::NewEnumerator1, Loop Behavior=::NewEnumerator1, Loop Duration=0.4, UseLoopDelay=false)
   &SpawnRate(SpawnRate=90, Use Spawn Probability=false)
  ]
  particleSpawnScript [
   &InitializeParticle(
    Color Maximum=<rgba 25 2.9167 0 0.75>
    Color Minimum=<rgba 25 4.1667 0 1>
    Color Mode=::NewEnumerator2
    Lifetime Max=0.1
    Lifetime Min=0.15
    Lifetime Mode=::NewEnumerator1
    Mass Mode=::NewEnumerator3
    Sprite Rotation Mode=::NewEnumerator2
    Sprite Size Mode=::NewEnumerator1
    UsePositionOffset=false
    Write Material Random=true
   )
   /SetVariables(Particles.MaterialRandom:NiagaraFloat=&RandomRangeFloat(Minimum=0, Maximum=1)) {isSetParameters=true, module=SetVariables_#1}
   &SphereLocation(
    MaskSpawnGroup=false
    Override Seed=false
    Sphere Radius=8
    Surface Only=false
    UseNonUniformScale=false
    UseOffset=false
    UseSphereOrientation=false
   ) {enabled=false}
   &AddVelocityInCone(
    Cone Angle=20
    Cone Axis=<0 0 1>
    Module.Use Velocity Falloff On Cone Axis=false
    Velocity Distribution Along Cone Axis=0.75
    Velocity Strength=&RandomRangeFloat(Minimum=500, Maximum=850)
   ) {enabled=false}
   /SetVariables(Particles.Herm Thickness:NiagaraFloat=&FloatFromCurve(
     FloatCurve=DataInterface() {propertyValues {Curve {keys [
         {arriveTangent=0.6937, interpMode=RCIM_Cubic, leaveTangent=0.6937, tangentMode=RCTM_User, time=0, value=0.075}
         {arriveTangent=-2.4361, interpMode=RCIM_Cubic, leaveTangent=-2.4361, tangentMode=RCTM_User, time=1, value=0}
        ]}}}
     CurveIndex=$Particles.MaterialRandom
    )) {isSetParameters=true, module=SetVariables_#2}
   &FN_HermiteSpline(
    Compute Acceleration=true
    Compute Velocity=true
    Direction End:Vector3f=&TransformVector(Vector=<-60 0 -80>, Destination Space=::Simulation)
    Direction Start:Vector3f=&TransformVector(Vector=<0 0 150>, Destination Space=::Simulation)
    Position Blend:NiagaraFloat=$Particles.MaterialRandom
    Position End:Vector3f=&TransformPosition(Position=<13 0 67>, Source Space=::Local, Destination Space=::Simulation)
    Position Start:Vector3f=&TransformPosition(Position=<0 0 -7>, Source Space=::Local, Destination Space=::Simulation)
    Write Particles Position=true
   )
   &SphereLocation(
    MaskSpawnGroup=false
    Non Uniform Scale=&MakeVector(X=$Particles.Herm Thickness, Y=0.1, Z=$Particles.Herm Thickness)
    Offset=<0 0.5 0>
    Override Seed=false
    Sphere Radius=1
    Surface Only=false
    UseNonUniformScale=true
    UseOffset=true
    UseSphereOrientation=false
   ) {module=SphereLocation001}
   &AddVelocityInCone(
    Cone Angle=20
    Cone Axis=<1 0 1>
    Module.Use Velocity Falloff On Cone Axis=false
    Velocity Strength=&Multiply_Float(A=&RandomRangeFloat(Minimum=300, Maximum=500), B=0.22)
   ) {enabled=false, module=AddVelocityInCone001}
   &InheritVelocity(Inherited Velocity Amount Scale=<0.1 0.1 0.1>, Speed Limit=false, UseSpeedThreshold=false) {enabled=false}
   &SubUVAnimation(End Frame=8, SubUV Animation Mode:ENiagara_SubUVLookupMode=::NewEnumerator2)
   &SpriteFacingAndAlignment(
    Sprite Alignment=&RandomVector(Vector Scale=1)
    Sprite Facing=&TransformVector(Vector=<0 1 0>, Destination Space=::Simulation)
    WriteAlignment=true
    WriteFacing=true
   )
   &AddVelocity(Velocity=<220 0 0>)
  ]
  particleUpdateScript [
   &ParticleState()
   &GravityForce() {enabled=false}
   &Drag(Drag=0.25, Use Linear Drag=true, Use Rotational Drag=false)
   &ScaleColor(
    Scale Alpha=&FloatFromCurve(FloatCurve=DataInterface() {propertyValues {Curve {keys [<at -0.0046 1.0113 RCIM_Cubic>,<at 1.0011 0.002 RCIM_Cubic>]}}})
    ScaleA=true
    ScaleRGB=true
   )
   &SolveForcesAndVelocity(Clamp Velocity=false, Limit Acceleration=false)
   &SpriteRotationRate(Rotation Rate=&RandomRangeFloat(Minimum=-360, Maximum=360))
  ]
  rendererClasses [&NiagaraSpriteRendererProperties]
  renderers [{rendererClass=&NiagaraSpriteRendererProperties}]
 }
 Woosh {
  emitterUpdateScript [
   &EmitterState(Life Cycle Mode=::NewEnumerator1, Loop Behavior=::NewEnumerator1, Loop Delay=0.025, Loop Duration=1, UseLoopDelay=true)
   &SpawnBurst_Instantaneous(Use Spawn Probability=false, UseLoopCountLimit=false)
  ]
  particleSpawnScript [&InitializeParticle(Color=<rgba 5 0.5833 0 0.1>, Lifetime=0.18, UsePositionOffset=false, Write Material Random=true)]
  particleUpdateScript [
   &ParticleState()
   &ScaleColor(
    Scale Alpha=&FloatFromCurve(FloatCurve=DataInterface() {propertyValues {Curve {keys [
         {interpMode=RCIM_Cubic, tangentMode=RCTM_User, time=0, value=0}
         {interpMode=RCIM_Cubic, tangentMode=RCTM_User, time=0.5, value=1}
         {tangentMode=RCTM_User, time=1, value=0}
        ]}}})
    ScaleA=true
    ScaleRGB=true
   )
  ]
  rendererClasses [&NiagaraMeshRendererProperties]
  renderers [{rendererClass=&NiagaraMeshRendererProperties, rendererIndex=0}]
 }
}
summary {
 systemName=NS_Pickaxe_Mystic_Male_Swing
 emitters [
  {bEnabled=false, emitterName=Ref, rendererClasses [&NiagaraComponentRendererProperties]}
  {bEnabled=false, emitterName=Sparks, rendererClasses [&NiagaraSpriteRendererProperties]}
  {emitterName=Symbols, rendererClasses [&NiagaraSpriteRendererProperties]}
  {emitterName=Woosh, rendererClasses [&NiagaraMeshRendererProperties]}
 ]
 userVariables [User.SkeletalMesh {
   type=NiagaraDataInterfaceSkeletalMesh
   defaultValue {
    struct=NiagaraExt_VariableValue_DataInterface
    value {dataInterface=&self:NiagaraDataInterfaceSkeletalMesh_0}
   }
  }]
}
