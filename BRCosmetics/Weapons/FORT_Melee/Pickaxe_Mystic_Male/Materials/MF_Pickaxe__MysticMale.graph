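# Material function for the Mystic Male pickaxe that overrides emissive color and opacity on incoming material attributes to create a glowing, animated reveal.
# MaterialCurveA sizes sphere masks in pre-skinned local space, with panning fire noise breaking up the edge and subtracted masks forming a bright rim.
# Layered, distorted swirl noises, gated by two mask textures, drive orange-red internal glow, with time of day blending toward night-glow parameters.
# Mask channels set per-region opacity inside the revealed area.
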MaterialFunction /BRCosmetics/Weapons/FORT_Melee/Pickaxe_Mystic_Male/Materials/MF_Pickaxe__MysticMale
graph {nodes [
  n1 = FunctionInput() {inputName=Material, inputType=FunctionInput_MaterialAttributes, sortPriority=2}
  n2 = Constant3Vector() {constant=<rgba 1500 287.8463 0 1>}
  n3 = LinearInterpolate(
   A=@n2
   B=Multiply(A=@n2, B=ScalarParameter() {parameterName=Global night glow}) {constB=0.7}
   Alpha=MaterialFunctionCall() {materialFunction=&Scalar_Time_of_Day}.Result
  )
  n4 = PreSkinnedPosition()
  n5 = ShaderStageSwitch(PixelShader=VertexInterpolator(VS=@n4).PS, VertexShader=@n4)
  n6 = Constant()
  n7 = TextureCoordinate()
  n8 = ScalarParameter() {defaultValue=1, parameterName=UVTiling_firemask2}
  n9 = TextureSample(UVs=Panner(Coordinate=Multiply(A=@n7, B=@n8)) {speedY=0.05}) {texture=/Game/Global/Textures/Default/Noise/T_FireNoise_01}
  n10 = TextureCoordinate()
  n11 = ScalarParameter() {defaultValue=1, parameterName=UVTiling_firemask}
  n12 = TextureSample(UVs=Panner(Coordinate=Multiply(A=@n10, B=@n11)) {speedY=0.05}) {texture=&T_FireNoise_01}
  n13 = Max(A=@n9.RGB, B=@n12.RGB)
  n14 = Add(A=@n6, B=Multiply(A=@n13) {constB=10})
  n15 = Multiply(A=ScalarParameter() {defaultValue=1, parameterName=MaterialCurveA}) {constB=100}
  n16 = Constant() {r=0.9}
  n17 = SphereMask(A=@n5, B=@n14, Radius=@n15, Hardness=@n16)
  n18 = Subtract(A=@n15)
  n19 = Subtract(A=Saturate(None=@n17), B=Saturate(None=SphereMask(A=@n5, B=@n14, Radius=@n18, Hardness=@n16)))
  n20 = Multiply(A=Multiply(A=TextureSample() {texture=/BRCosmetics/Weapons/FORT_Melee/Pickaxe_Mystic_Male/Textures/T_Pickaxe_Mystic_Mask}.R, B=@n19)) {constB=20}
  n21 = Multiply(A=@n17, B=Power(Base=ComponentMask(None=@n13) {r=true}) {constExponent=4})
  n22 = Multiply(A=@n19, B=@n21)
  n23 = Multiply(A=@n22) {constB=20}
  n24 = LinearInterpolate(A=Constant(), B=Constant() {r=1}, Alpha=Saturate(None=Add(A=@n20, B=@n23)))
  n25 = TextureSample() {texture=/BRCosmetics/Weapons/FORT_Melee/Pickaxe_Mystic_Male/Textures/T_Pickaxe_Mystic_Mask2}
  n26 = LinearInterpolate(
   B=Constant() {r=0.002}
   Alpha=Multiply(A=@n25.G, B=TextureSample() {samplerType=SAMPLERTYPE_LinearGrayscale, texture=/Game/Global/Textures/Default/Noise/T_SwirlNoise_N_height}.R)
  )
  n27 = TextureCoordinate() {uTiling=2, vTiling=2}
  n28 = TextureCoordinate() {uTiling=10, vTiling=10}
  n29 = Constant()
  n30 = TextureSample(UVs=Panner(Coordinate=MaterialFunctionCall(UVs=@n28, "Rotation Angle (0-1)"=@n29) {materialFunction=&CustomRotator}.Rotated Values) {speedX=0.1, speedY=0.1}) {samplerType=SAMPLERTYPE_LinearGrayscale, texture=&T_SwirlNoise_N_height}
  n31 = Panner(Coordinate=Add(A=TextureCoordinate() {uTiling=4, vTiling=4}, B=Multiply(A=@n30.R) {constB=0.02})) {speedX=0.01, speedY=0.01}
  n32 = TextureCoordinate() {uTiling=5, vTiling=5}
  n33 = TextureSample(UVs=Panner(Coordinate=MaterialFunctionCall(UVs=@n32, "Rotation Angle (0-1)"=@n29) {materialFunction=&CustomRotator}.Rotated Values) {speedX=0.5, speedY=0.5}) {samplerType=SAMPLERTYPE_LinearGrayscale, texture=&T_SwirlNoise_N_height}
  n34 = Panner(Coordinate=Add(A=TextureCoordinate(), B=Multiply(A=@n33.R) {constB=0.02})) {speedX=0.03, speedY=0.03}
  n35 = Multiply(
   A=TextureSample(UVs=Panner(Coordinate=@n27) {speedX=0.005, speedY=0.005}) {samplerType=SAMPLERTYPE_LinearColor, texture=/Game/Global/Textures/Default/Noise/T_Curie_NoisePack}.R
   B=MaterialFunctionCall(
    Base=TextureSample(UVs=@n31) {texture=/Game/Global/Textures/Default/Noise/Noise}.R
    Blend=TextureSample(UVs=@n34) {texture=&Noise}.R
   ) {materialFunction=&Blend_Overlay}.Result
  )
  n36 = Power(Base=@n35, Exp=ScalarParameter() {defaultValue=3, parameterName=internal glow threshold}) {constExponent=3}
  n37 = ScalarParameter() {defaultValue=0.2, parameterName=internal glow}
  n38 = LinearInterpolate(
   A=@n37
   B=Multiply(A=@n37, B=ScalarParameter() {parameterName=Global night glow})
   Alpha=MaterialFunctionCall() {materialFunction=&Scalar_Time_of_Day}.Result
  )
  n39 = Constant() {r=0.0015}
  n40 = TextureSample() {texture=&T_Pickaxe_Mystic_Mask}
  n41 = TextureCoordinate() {vTiling=0.3}
  n42 = Constant() {r=-0.3}
  n43 = TextureSample(UVs=Panner(Coordinate=MaterialFunctionCall(UVs=@n41, "Rotation Angle (0-1)"=@n42) {materialFunction=&CustomRotator}.Rotated Values) {speedX=0.2, speedY=0.2}) {samplerType=SAMPLERTYPE_LinearGrayscale, texture=&T_SwirlNoise_N_height}
  n44 = Multiply(A=Power(Base=@n40.G) {constExponent=1.8}, B=Power(Base=@n43.R))
  n45 = Constant() {r=15}
  n46 = LinearInterpolate(Alpha=Multiply(A=Power(Base=@n43.R) {constExponent=7}, B=@n40.R)) {constB=5}
  n47 = TextureCoordinate() {uTiling=2}
  n48 = Constant() {r=-0.3}
  n49 = TextureSample(UVs=Panner(Coordinate=MaterialFunctionCall(UVs=@n47, "Rotation Angle (0-1)"=@n48) {materialFunction=&CustomRotator}.Rotated Values) {speedX=0.1, speedY=0.1}) {samplerType=SAMPLERTYPE_LinearGrayscale, texture=&T_SwirlNoise_N_height}
  n50 = Multiply(A=Power(Base=@n40.B, Exp=LinearInterpolate(Alpha=@n49.R) {constA=1, constB=2}), B=Constant() {r=0.3}) {constB=0.5}
  n51 = Add(A=Max(A=Multiply(A=@n36, B=@n38), B=Multiply(A=@n35, B=@n39)), B=Add(A=Multiply(A=@n44, B=@n45) {constB=20}, B=Add(A=@n46, B=@n50)))
  n52 = Add(
   A=LinearInterpolate(A=Constant(), B=@n3, Alpha=Saturate(None=@n24))
   B=LinearInterpolate(A=Constant3Vector() {constant=<rgba 0.3 0.0029 0 1>}, B=@n3, Alpha=Add(A=@n26, B=@n51))
  )
  n53 = Saturate(None=@n25.A)
  n54 = LinearInterpolate(Alpha=@n53) {constA=1, constB=0}
  n55 = Constant() {r=0.4}
  n56 = LinearInterpolate(A=LinearInterpolate(A=LinearInterpolate(A=@n55, Alpha=@n53) {constB=0}, Alpha=@n25.R) {constB=0.2}, B=Constant() {r=0.05}, Alpha=@n25.B)
  n57 = Add(A=Multiply(A=@n13) {constB=3}, B=@n6)
  n58 = Subtract(A=@n15) {constB=2}
  n59 = Constant()
  n60 = Add(A=@n22, B=@n21)
  n61 = Saturate(None=Max(A=SphereMask(A=@n5, B=@n57, Radius=@n58, Hardness=@n59), B=ComponentMask(None=@n60) {r=true}))
  FunctionOutput(None=SetMaterialAttributes!(
    MaterialAttributes=GetMaterialAttributes(None=@n1).MaterialAttributes
    EmissiveColor=Multiply(A=@n52, B=@n54)
    Opacity=LinearInterpolate(B=@n56, Alpha=@n61)
   ) {attributeSetTypes [#1, #2]})
  MaterialFunctionCall() {materialFunction=&Color_Time_of_Day}
  MaterialFunctionCall(Base=@n9.RGB, Blend=@n12.RGB) {materialFunction=&Blend_Overlay}
  ScalarParameter() {defaultValue=100, parameterName=MaterialCurveA}
 ]}
