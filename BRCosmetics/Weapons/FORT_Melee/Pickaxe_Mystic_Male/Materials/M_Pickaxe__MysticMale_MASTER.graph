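# Master translucent, lit material for the Mystic Male pickaxe cosmetic, enabled for skeletal meshes and mesh deformers.
# It uses material attributes, passing the shared character PreFX result through the pickaxe-specific MF_Pickaxe__MysticMale function into the shared PostFX function.
# A separate SetMaterialAttributes node blends base color from red to blue by the MaterialCurveA scalar parameter, with no link shown to that chain.
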Material /BRCosmetics/Weapons/FORT_Melee/Pickaxe_Mystic_Male/Materials/M_Pickaxe__MysticMale_MASTER
graph {
 properties {
  bUseMaterialAttributes=true
  bUsedWithMeshDeformer=true
  bUsedWithSkeletalMesh=true
  blendMode=BLEND_Translucent
  pixelDepthOffsetMode=PDOM_Legacy
  usedShadingModels=Default Lit
 }
 nodes [
  n1 = MaterialFunctionCall() {materialFunction=&BaseFN_CharacterFunction_PreFX}
  MaterialFunctionCall(
   AO=@n1.AO
   Distance Blend=@n1.AO
   Fresnel=@n1.AO
   Gmap=@n1.AO
   Pre Skinned Local Position=@n1.AO
   Material=MaterialFunctionCall!(Material=@n1.Result) {materialFunction=&/MF_Pickaxe__MysticMale}.Result
  ) {materialFunction=&BaseFN_CharacterFunction_PostFX}
  SetMaterialAttributes!(
   MaterialAttributes
   Base Color=LinearInterpolate(
    A=Constant3Vector() {constant=<rgba 1 0 0 0>}
    B=Constant3Vector() {constant=<rgba 0 0 1 0>}
    Alpha=ScalarParameter() {parameterName=MaterialCurveA}
   )
  ) {attributeSetTypes [#1]}
 ]
}
