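# Physics asset for the Mystic Male pickaxe, a melee weapon cosmetic mesh.
# It holds a single capsule body on the Pickaxe_Mystic_Male_MESH_LOD0 bone, offset along Z and rotated to follow the length of the tool.
# It has no constraints, so the pickaxe simulates and collides as one rigid piece.
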PhysicsAsset /BRCosmetics/Weapons/FORT_Melee/Pickaxe_Mystic_Male/Meshes/Mystic_Male_PhysicsAsset
bodies {Pickaxe_Mystic_Male_MESH_LOD0 [{shapeType=Capsule, length=39.6686, radius=7.7417, center=<1.8367 0.4526 28.9596>, rotation=<rot -0.0539 -90 5.686>}]}
constraints []
